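; MaskMaker_PatrolPointProvider is a spawn point provider that configures the Cherry Smoke Mask Maker boss pawn when it spawns.
; On BeginPlay, only with network authority, it binds a handler to its actor-spawned result event.
; For a successful spawn, the handler casts the actor to the boss pawn and passes it this provider's portal target point and teleport area box.
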(event EventBeginPlay
  (Networking|SwitchHasAuthority
    (:Authority
      (EventDispatchers|BindEventtoOnActorSpawnedResult (delegate Custom|ActorSpawnedEvent)))))

(event Custom|ActorSpawnedEvent
  (params (SpawnedActor :type Actor%20Object%20Reference) (bResult :type Boolean))
  (if bResult
    (bind _asbp_player_pawn_cherry_smoke_mask_maker_boss (Utilities|Casting|CastToBP_PlayerPawn_CherrySmoke_MaskMakerBoss SpawnedActor))
    (Variables|/MaskMakerBoss/Pawn/BP_PlayerPawn_CherrySmoke_MaskMakerBoss.BP_PlayerPawn_CherrySmoke_MaskMakerBoss_C|SetPortalTargetPoint (Variables|Default|GetPortalTargetPoint) _asbp_player_pawn_cherry_smoke_mask_maker_boss)
    (Variables|/MaskMakerBoss/Pawn/BP_PlayerPawn_CherrySmoke_MaskMakerBoss.BP_PlayerPawn_CherrySmoke_MaskMakerBoss_C|SetTeleportAreaBox (Variables|Default|GetTeleportAreaBox) _asbp_player_pawn_cherry_smoke_mask_maker_boss)))

(fn ConstructionScript)
